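; BP_Cvar_Scalability_Debug is a debug Blueprint that applies custom console variable values for the current quality level.
; Its custom Execute event loops over the keys of the Custom Cvar Scalability map, each key being a cvar name.
; A Select on Current Quality picks one of five tier values, up to Cinematic, from that key's scalability options struct.
; It runs name and value as a console command through local player controller 0.
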(event Custom|Execute
  (bind _keys (Utilities|Map|Keys (Variables|Default|GetCustom%20Cvar%20Scalability)))
  (for _array_element _keys
    (Development|ExecuteConsoleCommand (Utilities|String|Append _array_element " " (Utilities|String|ToString(Float) (Utilities|Select :Index (Variables|Default|GetCurrent%20Quality) :Option%200 (Utilities|Struct|BreakBPScalabilityOptions (Utilities|Map|Find (Variables|Default|GetCustom%20Cvar%20Scalability) _array_element)) :Option%201 (.Medium_4_C892C2E1402D3D7AE28994887678785C (Utilities|Struct|BreakBPScalabilityOptions (Utilities|Map|Find (Variables|Default|GetCustom%20Cvar%20Scalability) _array_element))) :Option%202 (.High_6_8FA60BBB43DAD8306A94B899F8EEA5D8 (Utilities|Struct|BreakBPScalabilityOptions (Utilities|Map|Find (Variables|Default|GetCustom%20Cvar%20Scalability) _array_element))) :Option%203 (.Epic_8_290A0BDD4D6685042E731A9B6DA8F306 (Utilities|Struct|BreakBPScalabilityOptions (Utilities|Map|Find (Variables|Default|GetCustom%20Cvar%20Scalability) _array_element))) :Option%204 (.Cinematic_10_DF20BE5B49A9CC319D44BE8E2731898B (Utilities|Struct|BreakBPScalabilityOptions (Utilities|Map|Find (Variables|Default|GetCustom%20Cvar%20Scalability) _array_element)))))) (Game|GetLocalPlayerControllerFromID 0))))

(fn ConstructionScript)
